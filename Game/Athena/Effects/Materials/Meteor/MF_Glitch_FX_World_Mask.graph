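# Material function for the Nexus meteor FX that outputs a world-space glitch mask and a camera-distance fade mask.
# Layered, time-panned sines of world position drive a sphere mask with flickering radius, whose length blends a flicker term scaled by MPC_Nexus Glitch.
# A static switch adds meteor waves: two collection-driven planar bands along MeteorDirection plus a continuous panning sine.
# The fade grows with pixel-to-camera distance.
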MaterialFunction /Game/Athena/Effects/Materials/Meteor/MF_Glitch_FX_World_Mask
graph {nodes [
  Divide(A=Time() {bOverride_Period=true, period=2})
  n1 = VectorParameter() {defaultValue=<rgba -58257 52127 52272 0>}
  n2 = VectorParameter() {parameterName=PlaneLocation, defaultValue=<rgba 24056 2399 4844 0>}
  n3 = CollectionParameter() {collection=/Game/Athena/Environments/Nexus/Materials/FXshader/MPC_Nexus, parameterName=MeteorWave1}
  n4 = WorldPosition()
  n5 = Normalize(VectorInput=Subtract(A=LinearInterpolate(A=@n1.RGB, B=@n2.RGB, Alpha=@n3), B=@n4.XYZ))
  n6 = VectorParameter() {parameterName=MeteorDirection, defaultValue=<rgba 0.7755 -0.4565 -0.4361 0>}
  n7 = OneMinus(None=Abs(None=DotProduct(A=@n5, B=@n6.RGB)))
  n8 = CollectionParameter() {collection=&MPC_Nexus, parameterName=MeteorWave2}
  n9 = Normalize(VectorInput=Subtract(A=LinearInterpolate(A=@n1.RGB, B=@n2.RGB, Alpha=@n8), B=@n4.XYZ))
  n10 = OneMinus(None=Abs(None=DotProduct(A=@n9, B=@n6.RGB)))
  n11 = Multiply(A=@n6.RGB, B=WorldPosition().XYZ)
  n12 = ScalarParameter() {defaultValue=6000, parameterName=WavePanSize}
  n13 = Add(
   A=Time()
   B=CollectionParameter() {collection=&MPC_Nexus, parameterName=MeteorWaveContinuousAdd}
  )
  n14 = ScalarParameter() {defaultValue=-3, parameterName=WavePanSpeed}
  n15 = Sine(None=Add(A=Divide(A=@n11, B=@n12) {constB=2000}, B=Multiply(A=@n13, B=@n14)))
  n16 = Power(Base=Saturate(None=DotProduct(A=@n6.RGB, B=@n15)))
  n17 = CollectionParameter() {collection=&MPC_Nexus, parameterName=MeteorWaveContinuous}
  n18 = Multiply(A=Add(A=Add(A=@n7, B=@n10), B=Multiply(A=@n16, B=@n17) {constB=0.5}), B=ScalarParameter() {defaultValue=2, parameterName=WaveBrightness}) {constB=2}
  n19 = Time()
  n20 = ScalarParameter() {defaultValue=1.7, parameterName=FlickerAddSpeed}
  n21 = Multiply(A=Sine(None=Multiply(A=@n19, B=@n20)) {period=0.1}, B=ScalarParameter() {defaultValue=0.2, parameterName=FlickerMult})
  n22 = Constant() {r=1}
  n23 = Divide(
   A=WorldPosition().XYZ
   B=Multiply(
    A=FunctionInput(Preview=@n22) {bUsePreviewValueAsDefault=true, inputName=MaskScaleMult, inputType=FunctionInput_Scalar}
    B=ScalarParameter() {defaultValue=6000, parameterName=WorldMaskScale}
   )
  )
  n24 = Sine(None=ComponentMask(None=@n23) {r=true}) {period=1.7}
  n25 = Multiply(A=Add(A=Multiply(A=@n24) {constB=0.25}, B=Time()), B=ScalarParameter() {defaultValue=3, parameterName=WaveSpeed})
  n26 = Sine(None=Add(A=Multiply(A=@n25) {constB=0.8}, B=@n23)) {period=0.51}
  n27 = LinearInterpolate(A=Sine(None=Add(A=@n25, B=@n23)) {period=0.8}, B=Rotator(Coordinate=ConstantBiasScale(None=@n26)) {centerX=0, centerY=0, speed=1})
  n28 = Add(A=Multiply(A=@n25) {constB=-0.25}, B=@n23)
  n29 = Multiply(A=Sine(None=@n28) {period=0.42}) {constB=0.6}
  n30 = CollectionParameter() {collection=&MPC_Nexus, parameterName=Glitch}
  n31 = ConstantBiasScale(None=Sine(None=@n28))
  n32 = ScalarParameter() {defaultValue=0.25, parameterName=PhaseSpeed}
  n33 = Add(A=Round(None=@n31), B=ScalarParameter() {defaultValue=12, parameterName=FlickerSpeed})
  n34 = LinearInterpolate(
   A=OneMinus(None=@n31)
   B=@n31
   Alpha=LinearInterpolate(
    A=MaterialFunctionCall(Speed=@n32) {materialFunction=&DebugTimeSine}.Result
    B=MaterialFunctionCall(Speed=@n33) {materialFunction=&DebugTimeSine}.Result
   ) {constAlpha=0.2}
  ) {constA=-1}
  n35 = ScalarParameter() {defaultValue=3, parameterName=WPOMaskRadius}
  n36 = SphereMask(
   A=ConstantBiasScale(None=Add(A=@n27, B=@n29))
   B=Constant()
   Radius=Add(A=Multiply(A=@n30) {constB=0.4}, B=Multiply(A=@n34, B=@n35) {constB=2})
  ) {attenuationRadius=2, hardnessPercent=0}
  n37 = Multiply(
   A=Add(
    A=CollectionParameter() {collection=&MPC_Nexus, parameterName=Glitch}
    B=ScalarParameter() {defaultValue=0.5, parameterName=MinGlitchiness}
   )
   B=LinearInterpolate(
    B=Add(A=@n21)
    Alpha=MaterialFunctionCall(Vector 3=@n36) {materialFunction=/Engine/Functions/Engine_MaterialFunctions02/Utility/VectorLength}.V3 Length
   )
  )
  FunctionOutput(None=StaticSwitchParameter(True=Add(A=@n18, B=@n37), False=@n37) {parameterName=WorldSpacePortalMask}) {outputName=WorldSpaceMask}
  n38 = Distance(A=WorldPosition().XYZ, B=CameraPositionWS())
  n39 = ScalarParameter() {defaultValue=80000, parameterName=WPOOffsetFadeDist}
  FunctionOutput(None=Saturate(None=Divide(A=@n38, B=@n39) {constB=5000})) {outputName=DistanceFadeMask, sortPriority=1}
 ]}
